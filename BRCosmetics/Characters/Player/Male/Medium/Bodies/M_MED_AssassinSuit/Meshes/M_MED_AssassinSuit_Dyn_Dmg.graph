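# Physics asset for the dynamic damage bones on the male medium Assassin Suit body, covering the left bicep, both calves, and the right elbow and shoulder.
# Each damage bone carries a small capsule paired with a tiny capsule on its matching _MSTR bone.
# A constraint joins each pair with wide swing and very narrow twist limits, so every piece hinges independently about its own anchor, with no links between pairs.
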PhysicsAsset /BRCosmetics/Characters/Player/Male/Medium/Bodies/M_MED_AssassinSuit/Meshes/M_MED_AssassinSuit_Dyn_Dmg
bodies {
 dyn_dmg_bicep_l_01 [{shapeType=Capsule, length=2, radius=1, center=<0.3 0 0>, rotation=<rot 90 0 0>}]
 dyn_dmg_bicep_l_01_MSTR [{shapeType=Capsule, length=0.1, radius=0.1, rotation=<rot 90 0 0>}]
 dyn_dmg_bicep_l_02 [{shapeType=Capsule, length=2, radius=1, center=<0.3 0 0>, rotation=<rot 90 0 0>}]
 dyn_dmg_bicep_l_02_MSTR [{shapeType=Capsule, length=0.1, radius=0.1, rotation=<rot 90 0 0>}]
 dyn_dmg_calf_l_01 [{shapeType=Capsule, length=2, radius=1, center=<0.3 0 0>, rotation=<rot 90 0 0>}]
 dyn_dmg_calf_l_01_MSTR [{shapeType=Capsule, length=0.1, radius=0.1, rotation=<rot 90 0 0>}]
 dyn_dmg_calf_l_02 [{shapeType=Capsule, length=2, radius=1, center=<0.3 0 0>, rotation=<rot 90 0 0>}]
 dyn_dmg_calf_l_02_MSTR [{shapeType=Capsule, length=0.1, radius=0.1, rotation=<rot 90 0 0>}]
 dyn_dmg_calf_r_01 [{shapeType=Capsule, length=2, radius=1, center=<0.3 0 0>, rotation=<rot 90 0 0>}]
 dyn_dmg_calf_r_01_MSTR [{shapeType=Capsule, length=0.1, radius=0.1, rotation=<rot 90 0 0>}]
 dyn_dmg_calf_r_02 [{shapeType=Capsule, length=2, radius=1, center=<0.3 0 0>, rotation=<rot 90 0 0>}]
 dyn_dmg_calf_r_02_MSTR [{shapeType=Capsule, length=0.1, radius=0.1, rotation=<rot 90 0 0>}]
 dyn_dmg_calf_r_03 [{shapeType=Capsule, length=2, radius=1, center=<0.3 0 0>, rotation=<rot 90 0 0>}]
 dyn_dmg_calf_r_03_MSTR [{shapeType=Capsule, length=0.1, radius=0.1, rotation=<rot 90 0 0>}]
 dyn_dmg_calf_r_04 [{shapeType=Capsule, length=2, radius=1, center=<0.3 0 0>, rotation=<rot 90 0 0>}]
 dyn_dmg_calf_r_04_MSTR [{shapeType=Capsule, length=0.1, radius=0.1, rotation=<rot 90 0 0>}]
 dyn_dmg_elbow_r_01 [{shapeType=Capsule, length=2, radius=1, center=<0.3 0 0>, rotation=<rot 90 0 0>}]
 dyn_dmg_elbow_r_01_MSTR [{shapeType=Capsule, length=0.1, radius=0.1, rotation=<rot 90 0 0>}]
 dyn_dmg_elbow_r_02 [{shapeType=Capsule, length=2, radius=1, center=<0.3 0 0>, rotation=<rot 90 0 0>}]
 dyn_dmg_elbow_r_02_MSTR [{shapeType=Capsule, length=0.1, radius=0.1, rotation=<rot 90 0 0>}]
 dyn_dmg_shoulder_r_01 [{shapeType=Capsule, length=2, radius=1, center=<0.3 0 0>, rotation=<rot 90 0 0>}]
 dyn_dmg_shoulder_r_01_MSTR [{shapeType=Capsule, length=0.1, radius=0.1, rotation=<rot 90 0 0>}]
 dyn_dmg_shoulder_r_02 [{shapeType=Capsule, length=2, radius=1, center=<0.3 0 0>, rotation=<rot 90 0 0>}]
 dyn_dmg_shoulder_r_02_MSTR [{shapeType=Capsule, length=0.1, radius=0.1, rotation=<rot 90 0 0>}]
 dyn_dmg_shoulder_r_03 [{shapeType=Capsule, length=2, radius=1, center=<0.3 0 0>, rotation=<rot 90 0 0>}]
 dyn_dmg_shoulder_r_03_MSTR [{shapeType=Capsule, length=0.1, radius=0.1, rotation=<rot 90 0 0>}]
 dyn_dmg_shoulder_r_04 [{shapeType=Capsule, length=2, radius=1, center=<0.3 0 0>, rotation=<rot 90 0 0>}]
 dyn_dmg_shoulder_r_04_MSTR [{shapeType=Capsule, length=0.1, radius=0.1, rotation=<rot 90 0 0>}]
}
constraints [
 {bone1Name=dyn_dmg_bicep_l_01, bone2Name=dyn_dmg_bicep_l_01_MSTR, swing1LimitDegrees=45, swing2LimitDegrees=45, twistLimitDegrees=2}
 {bone1Name=dyn_dmg_bicep_l_02, bone2Name=dyn_dmg_bicep_l_02_MSTR, swing1LimitDegrees=45, swing2LimitDegrees=45, twistLimitDegrees=2}
 {bone1Name=dyn_dmg_elbow_r_01, bone2Name=dyn_dmg_elbow_r_01_MSTR, swing1LimitDegrees=45, swing2LimitDegrees=45, twistLimitDegrees=2}
 {bone1Name=dyn_dmg_shoulder_r_04, bone2Name=dyn_dmg_shoulder_r_04_MSTR, swing1LimitDegrees=45, swing2LimitDegrees=45, twistLimitDegrees=2}
 {bone1Name=dyn_dmg_shoulder_r_03, bone2Name=dyn_dmg_shoulder_r_03_MSTR, swing1LimitDegrees=45, swing2LimitDegrees=45, twistLimitDegrees=2}
 {bone1Name=dyn_dmg_elbow_r_02, bone2Name=dyn_dmg_elbow_r_02_MSTR, swing1LimitDegrees=45, swing2LimitDegrees=45, twistLimitDegrees=2}
 {bone1Name=dyn_dmg_shoulder_r_02, bone2Name=dyn_dmg_shoulder_r_02_MSTR, swing1LimitDegrees=45, swing2LimitDegrees=45, twistLimitDegrees=2}
 {bone1Name=dyn_dmg_shoulder_r_01, bone2Name=dyn_dmg_shoulder_r_01_MSTR, swing1LimitDegrees=45, swing2LimitDegrees=45, twistLimitDegrees=2}
 {bone1Name=dyn_dmg_calf_l_01, bone2Name=dyn_dmg_calf_l_01_MSTR, swing1LimitDegrees=45, swing2LimitDegrees=45, twistLimitDegrees=2}
 {bone1Name=dyn_dmg_calf_l_02, bone2Name=dyn_dmg_calf_l_02_MSTR, swing1LimitDegrees=45, swing2LimitDegrees=45, twistLimitDegrees=2}
 {bone1Name=dyn_dmg_calf_r_01, bone2Name=dyn_dmg_calf_r_01_MSTR, swing1LimitDegrees=45, swing2LimitDegrees=45, twistLimitDegrees=2}
 {bone1Name=dyn_dmg_calf_r_02, bone2Name=dyn_dmg_calf_r_02_MSTR, swing1LimitDegrees=45, swing2LimitDegrees=45, twistLimitDegrees=2}
 {bone1Name=dyn_dmg_calf_r_03, bone2Name=dyn_dmg_calf_r_03_MSTR, swing1LimitDegrees=45, swing2LimitDegrees=45, twistLimitDegrees=2}
 {bone1Name=dyn_dmg_calf_r_04, bone2Name=dyn_dmg_calf_r_04_MSTR, swing1LimitDegrees=45, swing2LimitDegrees=45, twistLimitDegrees=2}
]
