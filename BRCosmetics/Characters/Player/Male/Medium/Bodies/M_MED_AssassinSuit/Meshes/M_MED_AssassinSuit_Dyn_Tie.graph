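# Physics asset for the necktie on the male medium AssassinSuit body cosmetic, giving the dyn_tie bones capsule bodies for physical motion.
# Five tie segments chain from dyn_tie_MSTR, which links to spine_05, with swing limits widening toward the tip and tight twist limits throughout.
# Larger capsules on the pelvis and five spine bones approximate the torso, joined by angularly free constraints, as is the tie's master bone.
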PhysicsAsset /BRCosmetics/Characters/Player/Male/Medium/Bodies/M_MED_AssassinSuit/Meshes/M_MED_AssassinSuit_Dyn_Tie
bodies {
 dyn_tie_01 [{shapeType=Capsule, length=6, radius=3, shapeName=dyn_tie_01_capsule, center=<4.805 0 1.5923>, rotation=<rot 90 0 0>}]
 dyn_tie_02 [{shapeType=Capsule, length=6, radius=3, shapeName=dyn_tie_02_capsule, center=<4.805 0 1.5079>, rotation=<rot 90 0.0099 0.0099>}]
 dyn_tie_03 [{shapeType=Capsule, length=6, radius=3, shapeName=dyn_tie_03_capsule, center=<4.805 0 2.0263>, rotation=<rot 90 0.0099 0.0099>}]
 dyn_tie_04 [{shapeType=Capsule, length=6, radius=3, shapeName=dyn_tie_04_capsule, center=<4.805 0 1.5136>, rotation=<rot 90 0 0>}]
 dyn_tie_05 [{shapeType=Capsule, length=1.25, radius=2, shapeName=dyn_tie_05_capsule, center=<2.4387 0 0.4718>, rotation=<rot 90 0 0>}]
 dyn_tie_MSTR [{shapeType=Capsule, length=1, radius=0.5, rotation=<rot 90 0 0>}]
 pelvis [{shapeType=Capsule, length=9, radius=8.5, shapeName=pelvis_capsule, center=<7.2135 0 -0.0001>}]
 spine_01 [{shapeType=Capsule, length=6.7188, radius=10.9688, center=<-2.0733 -1.3709 0>}]
 spine_02 [{shapeType=Capsule, length=19.1562, radius=12.0938, center=<0 -0.6181 0>}]
 spine_03 [{shapeType=Capsule, length=19.1562, radius=12.0938, center=<0 -0.2866 0>}]
 spine_04 [{shapeType=Capsule, length=19.1562, radius=12.0938, center=<0 -0.1776 0>}]
 spine_05 [{shapeType=Capsule, length=20, radius=13.2188, center=<-4.5203 -1.893 0>}]
}
constraints [
 {bone1Name=dyn_tie_01, bone2Name=dyn_tie_MSTR, swing1LimitDegrees=20, swing2LimitDegrees=15, twistLimitDegrees=5}
 {bone1Name=dyn_tie_02, bone2Name=dyn_tie_01, swing1LimitDegrees=20, swing2LimitDegrees=15, twistLimitDegrees=5}
 {bone1Name=dyn_tie_03, bone2Name=dyn_tie_02, swing1LimitDegrees=30, swing2LimitDegrees=15, twistLimitDegrees=5}
 {bone1Name=dyn_tie_04, bone2Name=dyn_tie_03, swing1LimitDegrees=60, swing2LimitDegrees=30, twistLimitDegrees=5}
 {bone1Name=dyn_tie_05, bone2Name=dyn_tie_04, swing1LimitDegrees=60, swing2LimitDegrees=30, twistLimitDegrees=5}
 {
  bone1Name=dyn_tie_MSTR
  bone2Name=spine_05
  swing1LimitDegrees=45
  swing1Motion=Free
  swing2LimitDegrees=45
  swing2Motion=Free
  twistLimitDegrees=45
  twistMotion=Free
 }
 {
  bone1Name=spine_05
  bone2Name=spine_04
  swing1LimitDegrees=45
  swing1Motion=Free
  swing2LimitDegrees=45
  swing2Motion=Free
  twistLimitDegrees=45
  twistMotion=Free
 }
 {
  bone1Name=spine_04
  bone2Name=spine_03
  swing1LimitDegrees=45
  swing1Motion=Free
  swing2LimitDegrees=45
  swing2Motion=Free
  twistLimitDegrees=45
  twistMotion=Free
 }
 {
  bone1Name=spine_03
  bone2Name=spine_02
  swing1LimitDegrees=45
  swing1Motion=Free
  swing2LimitDegrees=45
  swing2Motion=Free
  twistLimitDegrees=45
  twistMotion=Free
 }
 {
  bone1Name=spine_02
  bone2Name=spine_01
  swing1LimitDegrees=45
  swing1Motion=Free
  swing2LimitDegrees=45
  swing2Motion=Free
  twistLimitDegrees=45
  twistMotion=Free
 }
 {
  bone1Name=spine_01
  bone2Name=pelvis
  swing1LimitDegrees=45
  swing1Motion=Free
  swing2LimitDegrees=45
  swing2Motion=Free
  twistLimitDegrees=45
  twistMotion=Free
 }
]
